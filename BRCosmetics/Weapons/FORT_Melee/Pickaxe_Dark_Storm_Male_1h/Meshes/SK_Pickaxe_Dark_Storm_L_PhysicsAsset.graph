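# Physics asset for the left-hand Dark Storm one-handed pickaxe skeletal mesh, a cosmetic melee weapon.
# It defines a single body on the Melee_TwoHanded_Handle bone, shaped as one offset, rotated capsule that stands in for the weapon's collision.
# It has no constraints, so the weapon has no jointed or chained physics parts.
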PhysicsAsset /BRCosmetics/Weapons/FORT_Melee/Pickaxe_Dark_Storm_Male_1h/Meshes/SK_Pickaxe_Dark_Storm_L_PhysicsAsset
bodies {Melee_TwoHanded_Handle [{shapeType=Capsule, length=38.2499, radius=15.0502, center=<-2.9637 6.212 -8.5623>, rotation=<rot -7.9748 -90 -7.284>}]}
constraints []
