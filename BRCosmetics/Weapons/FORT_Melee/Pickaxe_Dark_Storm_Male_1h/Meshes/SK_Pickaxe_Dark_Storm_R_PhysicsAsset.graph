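# Physics asset for the right-hand skeletal mesh of the Dark Storm one-handed male pickaxe cosmetic in Battle Royale.
# It defines a single angled capsule body on the b_AR_Root bone as the weapon's simple collision and physics shape.
# There are no constraints, so the mesh has no jointed or articulated physics parts.
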PhysicsAsset /BRCosmetics/Weapons/FORT_Melee/Pickaxe_Dark_Storm_Male_1h/Meshes/SK_Pickaxe_Dark_Storm_R_PhysicsAsset
bodies {b_AR_Root [{shapeType=Capsule, length=38.2499, radius=15.0024, center=<6.204 -2.8857 -8.5534>, rotation=<rot -7.3544 -90 -7.91>}]}
constraints []
